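# Material function in the MaterialGym RayMarchedNoise set that supplies a texture-based noise value, named as an SDF for ray marching.
# It samples VT_ClothSimWPO at an optional Vector input, which defaults to scaled-down world position.
# The red channel is offset by a fixed threshold and floored at zero, so the output is never negative.
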MaterialFunction /Game/Global/Materials/MaterialGym/Materials/RayMarchedNoise/MaterialFunctions/MF_RayMarchedNoiseSDF
graph {nodes [
  n1 = WorldPosition()
  n2 = TextureSample(UVs=FunctionInput(Preview=Divide(A=@n1.XYZ) {constB=100}) {bUsePreviewValueAsDefault=true, inputName=Vector}) {samplerType=SAMPLERTYPE_LinearColor, texture=/Game/Global/MaterialFunctions/VT_ClothSimWPO}
  FunctionOutput(None=Max(A=Subtract(A=@n2.R) {constB=0.5}) {constB=0})
 ]}
